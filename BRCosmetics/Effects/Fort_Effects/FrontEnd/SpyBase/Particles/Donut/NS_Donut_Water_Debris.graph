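# Niagara system for the SpyBase front-end "Donut" scene that fills a box of water with drifting bubbles and floating paper debris.
# Two sprite emitters spawn small deep bubbles and larger surface bubbles that drift upward, while two Papers emitters spawn sparse, long-lived, slowly tumbling meshes that scale in and out.
# All four use probability-gated spawn rates and a local LimitPositionAboveZero input that constrains particle position each update.
# None of the declared User parameters are referenced by the visible modules.
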NiagaraSystem /BRCosmetics/Effects/Fort_Effects/FrontEnd/SpyBase/Particles/Donut/NS_Donut_Water_Debris
dependencies {usedDataInterfaces [&NiagaraDataInterfaceCurve], usedDynamicInputs [], usedModules [&SystemState]}
emitters {
 Bubbles {
  emitterUpdateScript [
   &EmitterState(Life Cycle Mode=::NewEnumerator1, Loop Duration=1, UseLoopDelay=false)
   &SpawnRate(Spawn Probability=0.25, SpawnRate=350, Use Spawn Probability=true)
  ]
  particleSpawnScript [
   /Niagara/Modules/Spawn/Initialization/InitializeParticle(
    Lifetime=4
    Module.Write Color=true
    Module.Write Lifetime=true
    Module.Write Mass=true
    Module.Write Material Random=true
    Module.Write Position=false
    Module.Write Scale=true
    Module.Write SpriteRotation=false
    Module.Write SpriteSize=true
    Position=&SimulationPosition()
    Sprite Size=&Vector2DFromFloat(Value=&UniformRangedFloat(Minimum=0.75, Maximum=2.25))
   )
   /SetVariables(Particles.RandomPosNeg:NiagaraFloat=&UniformRangedFloat(Minimum=-1, Maximum=1)) {isSetParameters=true, module=SetVariables_#1}
   &BoxLocation(Box Offset=<0 0 -50>, Box Size=<500 200 100>, MaskSpawnGroup=false, Module.Override Seed=false, Rotate Box=false, Surface Only=false)
   &AddVelocity(Velocity=&UniformRangedVector(Minimum=<4 4 8>, Maximum=<-4 -4 4>))
  ]
  particleUpdateScript [
   &ParticleState()
   &SolveForcesAndVelocity(Clamp Velocity=false, Limit Acceleration=false)
   /Niagara/Modules/Update/Orientation/V2/UpdateMeshOrientation(
    Rotation Rate:NiagaraFloat=&Multiply_Float(A=0.035, B=$Particles.RandomPosNeg)
    Rotation Vector:Vector3f=&VectorFromFloat(Value=1)
   )
   /SetVariables(Particles.Position=&self:LimitPositionAboveZero*()) {isSetParameters=true, module=SetVariables_#2}
   &ScaleSpriteSize(Scale Factor=&Vector2DFromFloat(Value=&FloatFromCurve(FloatCurve=DataInterface() {propertyValues {Curve {keys [
          <at 0 0 RCIM_Cubic>
          {arriveTangent=0.002, interpMode=RCIM_Cubic, leaveTangent=0.002, time=0.2595, value=0.9859}
          <at 1.0011 0.002 RCIM_Cubic>
         ]}}})))
  ]
  rendererClasses [&NiagaraSpriteRendererProperties]
  renderers [{rendererClass=&NiagaraSpriteRendererProperties}]
 }
 BubblesSurface {
  emitterUpdateScript [
   &EmitterState(Life Cycle Mode=::NewEnumerator1, Loop Duration=1, UseLoopDelay=false)
   &SpawnRate(Spawn Probability=0.25, SpawnRate=150, Use Spawn Probability=true)
  ]
  particleSpawnScript [
   &InitializeParticle(
    Lifetime=3
    Module.Write Color=true
    Module.Write Lifetime=true
    Module.Write Mass=true
    Module.Write Material Random=true
    Module.Write Position=false
    Module.Write Scale=false
    Module.Write SpriteRotation=false
    Module.Write SpriteSize=true
    Position=&SimulationPosition()
    Sprite Size=&Vector2DFromFloat(Value=&UniformRangedFloat(Minimum=2.25, Maximum=4))
   )
   /SetVariables(Particles.RandomPosNeg:NiagaraFloat=&UniformRangedFloat(Minimum=-1, Maximum=1)) {isSetParameters=true, module=SetVariables_#1}
   &BoxLocation(Box Size=<500 200 25>, MaskSpawnGroup=false, Module.Override Seed=false, Rotate Box=false, Surface Only=false)
   &AddVelocity(Velocity=&UniformRangedVector(Minimum=<8 8 25>, Maximum=<-8 -8 12>))
   /SetVariables() {isSetParameters=true}
  ]
  particleUpdateScript [
   &ParticleState()
   &SolveForcesAndVelocity(Clamp Velocity=false, Limit Acceleration=false)
   &UpdateMeshOrientation(
    Rotation Rate:NiagaraFloat=&Multiply_Float(A=0.035, B=$Particles.RandomPosNeg)
    Rotation Vector:Vector3f=&VectorFromFloat(Value=1)
   )
   /SetVariables(Particles.Position=&self:LimitPositionAboveZero*()) {isSetParameters=true, module=SetVariables_#2}
   &ScaleSpriteSize(Scale Factor=&Vector2DFromFloat(Value=&FloatFromCurve(FloatCurve=DataInterface() {propertyValues {Curve {keys [
          <at 0 0 RCIM_Cubic>
          {interpMode=RCIM_Cubic, tangentMode=RCTM_Break, time=0.831, value=0.7169}
          {arriveTangent=2.4154, interpMode=RCIM_Cubic, leaveTangent=2.4154, tangentMode=RCTM_User, time=1, value=1}
         ]}}})))
  ]
  rendererClasses [&NiagaraSpriteRendererProperties]
  renderers [{rendererClass=&NiagaraSpriteRendererProperties}]
 }
 Papers_01 {
  emitterUpdateScript [
   &EmitterState(Life Cycle Mode=::NewEnumerator1, Loop Duration=1, UseLoopDelay=false)
   &SpawnRate(Spawn Probability=0.25, SpawnRate=2, Use Spawn Probability=true)
  ]
  particleSpawnScript [
   &InitializeParticle(
    Lifetime=15
    Mesh Scale=&VectorFromFloat(Value=&UniformRangedFloat(Minimum=0.125, Maximum=0.35))
    Module.Write Color=true
    Module.Write Lifetime=true
    Module.Write Mass=true
    Module.Write Material Random=true
    Module.Write Position=false
    Module.Write Scale=true
    Module.Write SpriteRotation=false
    Module.Write SpriteSize=true
    Position=&SimulationPosition()
    Sprite Size=<5 5>
   )
   /SetVariables(Particles.RandomPosNeg:NiagaraFloat=&UniformRangedFloat(Minimum=-1, Maximum=1)) {isSetParameters=true, module=SetVariables_#1}
   /Niagara/Modules/Spawn/Orientation/V2/InitialMeshOrientation(
    Rotation=&UniformRangedVector(Minimum=<0 0 0>, Maximum=<1 1 1>)
    Use Orientation Vector=true
    Use Rotation Vector=true
   )
   &BoxLocation(Box Offset=<0 0 -50>, Box Size=<400 175 100>, MaskSpawnGroup=false, Module.Override Seed=false, Rotate Box=false, Surface Only=false)
   &AddVelocity(Velocity=&UniformRangedVector(Minimum=<8 8 8>, Maximum=<-8 -8 -8>))
  ]
  particleUpdateScript [
   &ParticleState()
   &SolveForcesAndVelocity(Clamp Velocity=false, Limit Acceleration=false)
   &UpdateMeshOrientation(
    Rotation Rate:NiagaraFloat=&Multiply_Float(A=0.05, B=$Particles.RandomPosNeg)
    Rotation Vector:Vector3f=&VectorFromFloat(Value=1)
   )
   &ScaleMeshSize(Scale Factor=&VectorFromFloat(Value=&FloatFromCurve(FloatCurve=DataInterface() {propertyValues {Curve {keys [
          <at 0 0 RCIM_Cubic>
          {arriveTangent=-0.0039, interpMode=RCIM_Cubic, leaveTangent=-0.0039, tangentMode=RCTM_User, time=0.3144, value=1}
          {arriveTangent=-0.0362, interpMode=RCIM_Cubic, leaveTangent=-0.0362, tangentMode=RCTM_User, time=0.6825, value=1}
          <at 1 0 RCIM_Cubic>
         ]}}})))
   /SetVariables(Particles.Position=&self:LimitPositionAboveZero*()) {isSetParameters=true, module=SetVariables_#2}
  ]
  rendererClasses [&NiagaraSpriteRendererProperties, &NiagaraMeshRendererProperties]
  renderers [
   {rendererClass=&NiagaraSpriteRendererProperties}
   {rendererClass=&NiagaraMeshRendererProperties, rendererIndex=1}
  ]
 }
 Papers_02 {
  emitterUpdateScript [
   &EmitterState(Life Cycle Mode=::NewEnumerator1, Loop Duration=1, UseLoopDelay=false)
   &SpawnRate(Spawn Probability=0.25, SpawnRate=2, Use Spawn Probability=true)
  ]
  particleSpawnScript [
   &InitializeParticle(
    Lifetime=15
    Mesh Scale=&VectorFromFloat(Value=&UniformRangedFloat(Minimum=0.2, Maximum=0.5))
    Module.Write Color=true
    Module.Write Lifetime=true
    Module.Write Mass=true
    Module.Write Material Random=true
    Module.Write Position=false
    Module.Write Scale=true
    Module.Write SpriteRotation=false
    Module.Write SpriteSize=true
    Position=&SimulationPosition()
    Sprite Size=<5 5>
   )
   /SetVariables(Particles.RandomPosNeg:NiagaraFloat=&UniformRangedFloat(Minimum=-1, Maximum=1)) {isSetParameters=true, module=SetVariables_#1}
   &InitialMeshOrientation(
    Rotation=&UniformRangedVector(Minimum=<0 0 0>, Maximum=<1 1 1>)
    Use Orientation Vector=true
    Use Rotation Vector=true
   )
   &BoxLocation(Box Offset=<0 0 -50>, Box Size=<350 100 100>, MaskSpawnGroup=false, Module.Override Seed=false, Rotate Box=false, Surface Only=false)
   &AddVelocity(Velocity=&UniformRangedVector(Minimum=<8 8 8>, Maximum=<-8 -8 -8>))
  ]
  particleUpdateScript [
   &ParticleState()
   &SolveForcesAndVelocity(Clamp Velocity=false, Limit Acceleration=false)
   &UpdateMeshOrientation(
    Rotation Rate:NiagaraFloat=&Multiply_Float(A=0.035, B=$Particles.RandomPosNeg)
    Rotation Vector:Vector3f=&VectorFromFloat(Value=1)
   )
   &ScaleMeshSize(Scale Factor=&VectorFromFloat(Value=&FloatFromCurve(FloatCurve=DataInterface() {propertyValues {Curve {keys [
          <at 0 0 RCIM_Cubic>
          {arriveTangent=-0.0039, interpMode=RCIM_Cubic, leaveTangent=-0.0039, tangentMode=RCTM_User, time=0.3412, value=1}
          {arriveTangent=-0.0362, interpMode=RCIM_Cubic, leaveTangent=-0.0362, tangentMode=RCTM_User, time=0.6682, value=1}
          <at 1 0 RCIM_Cubic>
         ]}}})))
   /SetVariables(Particles.Position=&self:LimitPositionAboveZero*()) {isSetParameters=true, module=SetVariables_#2}
  ]
  rendererClasses [&NiagaraSpriteRendererProperties, &NiagaraMeshRendererProperties]
  renderers [
   {rendererClass=&NiagaraSpriteRendererProperties}
   {rendererClass=&NiagaraMeshRendererProperties, rendererIndex=1}
  ]
 }
}
summary {
 systemName=NS_Donut_Water_Debris
 emitters [
  {emitterName=Papers_01, rendererClasses [&NiagaraSpriteRendererProperties, &NiagaraMeshRendererProperties]}
  {emitterName=Papers_02, rendererClasses [&NiagaraSpriteRendererProperties, &NiagaraMeshRendererProperties]}
  {emitterName=Bubbles, rendererClasses [&NiagaraSpriteRendererProperties]}
  {emitterName=BubblesSurface, rendererClasses [&NiagaraSpriteRendererProperties]}
 ]
 userVariables [
  User.Altitude {type=NiagaraFloat, defaultValue=0, description=""}
  User.BubbleOffset {type=NiagaraFloat, defaultValue=0, description=""}
  User.BubblesOffset {type=NiagaraFloat, defaultValue=0, description=""}
  User.hasDriver {type=NiagaraBool}
  User.ImpactNormal {type=Vector3f, description="", defaultValue {struct=Vector3f, value=<0 0 0>}}
  User.PhysMatIndex {type=NiagaraInt32}
  User.PreviousVelocityXY {type=Vector3f, description="", defaultValue {struct=Vector3f, value=<0 0 0>}}
  User.RotorPlaySpeed {type=NiagaraFloat, defaultValue=0, description=""}
  User.Theta {type=NiagaraFloat, defaultValue=0, description=""}
  User.TorusInnerRadius {type=NiagaraFloat, defaultValue=0, description=""}
  User.Width {type=NiagaraFloat, defaultValue=0, description=""}
 ]
}
