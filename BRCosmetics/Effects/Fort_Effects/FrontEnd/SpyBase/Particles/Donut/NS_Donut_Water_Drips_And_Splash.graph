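# Niagara water drip and splash effect for the Donut part of the SpyBase front end, exposing a DonutWaterLocation user parameter.
# DrippingWater randomly spawns falling sprites across a flat box, kills them through a DestroyAtVector check and generates death events.
# Short-lived crown splash meshes, two sizes of arcing droplets and curl-noise underwater bubbles form the impact, though their event hookup is not shown.
# A Ripple emitter driven by animated material parameters is present but disabled.
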NiagaraSystem /BRCosmetics/Effects/Fort_Effects/FrontEnd/SpyBase/Particles/Donut/NS_Donut_Water_Drips_And_Splash
dependencies {
 usedDataInterfaces [&NiagaraDataInterfaceCurve, &NiagaraDataInterfaceColorCurve, &NiagaraDataInterfaceVectorCurve]
 usedDynamicInputs []
 usedModules [&SystemState]
}
emitters {
 CrownSplashMesh {
  emitterUpdateScript [&EmitterState()]
  particleSpawnScript [
   /Niagara/Modules/Spawn/Initialization/InitializeParticle(
    Lifetime=0.5
    Module.Write Color=true
    Module.Write Lifetime=true
    Module.Write Mass=true
    Module.Write Material Random=true
    Module.Write Position=true
    Module.Write Scale=true
    Module.Write SpriteRotation=false
    Module.Write SpriteSize=true
    Position=&SimulationPosition()
    Sprite Size=&Multiply_Vector2(
     A=<3 5.5>
     B=&Vector2DFromFloat(Value=&UniformRangedFloat(Minimum=0.1, Maximum=2.25))
    )
   )
   &AddVelocityInCone(
    Cone Angle=80
    Cone Axis=<0 0 1>
    Module.Use Velocity Falloff On Cone Axis=false
    Velocity Strength=/Game/Effects/Niagara/DynamicInputs/FN_PickFloatFromSpawnGroupCutoff(
     Over Cutoff Float:NiagaraFloat=0
     Under Cutoff Float:NiagaraFloat=&UniformRangedFloat(Minimum=50, Maximum=250)
     SpawnGroup:NiagaraInt32=$Emitter.SpawnGroup
     Cutoff:NiagaraInt32=2
    )
   ) {enabled=false}
   /SetVariables() {isSetParameters=true, module=SetVariables_#1}
   /SetVariables() {isSetParameters=true, module=SetVariables_#2}
   /Niagara/Modules/Spawn/Orientation/V2/InitialMeshOrientation(
    Rotation=&UniformRangedVector(Minimum=<0 0 0>, Maximum=<0 0 1>)
    Use Orientation Vector=true
    Use Rotation Vector=true
   )
  ]
  particleUpdateScript [
   &ParticleState()
   &ScaleColor(
    Scale Alpha=&FloatFromCurve(FloatCurve=DataInterface() {propertyValues {Curve {keys [<at 0.65 1.0113>,<at 1.0011 0.002>]}}})
    ScaleA=true
    ScaleRGB=true
   )
   /SetVariables(Particles.MeshColor:LinearColor=&ScaleLinearColorByCurve(
     InitialValue=<rgba 1 1 1 2>
     ScaleCurve=DataInterface() {propertyValues {
       AlphaCurve {keys [<at 0.2012 1>,<at 0.5654 0>]}
       BlueCurve {keys [<at 0.3226 1>,<at 0.9987 0>]}
       GreenCurve {keys [<at 0.3226 1>,<at 0.9987 0>]}
       RedCurve {keys [<at 0.3226 1>,<at 0.9987 0>]}
      }}
     CurveIndex=$Particles.NormalizedAge
    )) {isSetParameters=true, module=SetVariables_#3}
   &Drag(Use Linear Drag=true, Use Rotational Drag=false)
   &SolveForcesAndVelocity(Clamp Velocity=false, Limit Acceleration=false)
   &ScaleMeshSize(Scale Factor=&VectorFromCurve(VectorCurve=DataInterface() {propertyValues {
       XCurve {keys [{interpMode=RCIM_Cubic, leaveTangent=4.8413, tangentMode=RCTM_Break, time=0, value=0}, <at 0.7127 0.5793>]}
       YCurve {keys [{interpMode=RCIM_Cubic, leaveTangent=4.8413, tangentMode=RCTM_Break, time=0, value=0}, <at 0.7127 0.5793>]}
       ZCurve {keys [
         {interpMode=RCIM_Cubic, leaveTangent=7.2542, tangentMode=RCTM_Break, time=0, value=0}
         {arriveTangent=-1.166, interpMode=RCIM_Cubic, leaveTangent=-1.166, tangentMode=RCTM_User, time=0.0765, value=1.1529}
         <at 0.9989 0>
        ]}
      }}))
   /Niagara/Modules/Update/Orientation/V2/UpdateMeshOrientation(Rotation Rate=0.2, Rotation Vector:Vector3f=<0 0 1>)
  ]
  rendererClasses [&NiagaraMeshRendererProperties]
  renderers [{rendererClass=&NiagaraMeshRendererProperties, rendererIndex=0}]
 }
 DrippingWater {
  emitterUpdateScript [
   &EmitterState(Life Cycle Mode=::NewEnumerator1, Loop Duration=1, UseLoopDelay=false)
   &SpawnRate(Spawn Probability=&UniformRangedFloat(Minimum=0, Maximum=1), SpawnRate=6, Use Spawn Probability=true)
  ]
  particleSpawnScript [
   &InitializeParticle(
    Color=<rgba 1 1 1 0.5>
    Module.Write Color=true
    Module.Write Lifetime=true
    Module.Write Mass=true
    Module.Write Material Random=true
    Module.Write Position=true
    Module.Write Scale=false
    Module.Write SpriteRotation=false
    Module.Write SpriteSize=true
    Position=&SimulationPosition()
    Sprite Size=<4 15>
   )
   &BoxLocation(Box Size=<380 225 0>, MaskSpawnGroup=false, Module.Override Seed=false, Rotate Box=false, Surface Only=false)
  ]
  particleUpdateScript [
   &ParticleState()
   &GravityForce()
   &Drag(Use Linear Drag=true, Use Rotational Drag=false)
   &SolveForcesAndVelocity(Clamp Velocity=false, Limit Acceleration=false)
   /SetVariables(DataInstance.Alive:NiagaraBool=&self:DestroyAtVector()) {isSetParameters=true}
   &GenerateDeathEvent(Gap Correction Amount=1.5)
  ]
  rendererClasses [&NiagaraSpriteRendererProperties]
  renderers [{rendererClass=&NiagaraSpriteRendererProperties}]
 }
 ImpactDroplets {
  emitterUpdateScript [&EmitterState()]
  particleSpawnScript [
   &InitializeParticle(
    Color=<rgba 1 1 1 0.2>
    Lifetime=&UniformRangedFloat(Minimum=0.2, Maximum=0.6)
    Module.Write Color=true
    Module.Write Lifetime=true
    Module.Write Mass=true
    Module.Write Material Random=true
    Module.Write Position=false
    Module.Write Scale=true
    Module.Write SpriteRotation=false
    Module.Write SpriteSize=true
    Position=&SimulationPosition()
    Sprite Size=&Multiply_Vector2(
     A=<3 5.5>
     B=&Vector2DFromFloat(Value=&UniformRangedFloat(Minimum=0.1, Maximum=2.25))
    )
   )
   &AddVelocityInCone(
    Cone Angle=80
    Cone Axis=<0 0 1>
    Module.Use Velocity Falloff On Cone Axis=false
    Velocity Strength=&FN_PickFloatFromSpawnGroupCutoff(
     Over Cutoff Float:NiagaraFloat=0
     Under Cutoff Float:NiagaraFloat=&UniformRangedFloat(Minimum=50, Maximum=250)
     SpawnGroup:NiagaraInt32=$Emitter.SpawnGroup
     Cutoff:NiagaraInt32=2
    )
   ) {enabled=false}
   /SetVariables(Particles.PositionSplash:NiagaraPosition=$Engine.Owner.Position) {isSetParameters=true, module=SetVariables_#4}
   /SetVariables(Particles.Scale=/Game/Effects/Niagara/DynamicInputs/Water/FN_PickVectorFromSpawnGroupCutoff(
     Under Cutoff Vector:Vector3f=<0 0 0>
     Over Cutoff Vector:Vector3f=<1 1 1>
     SpawnGroup:NiagaraInt32=$Emitter.SpawnGroup
     Cutoff:NiagaraInt32=2
    )) {isSetParameters=true, module=SetVariables_#1}
   /SetVariables() {isSetParameters=true, module=SetVariables_#2}
   &CylinderLocation(
    Cylinder Height=0
    Cylinder Radius=8
    Lathe Profile=DataInterface() {propertyValues {Curve {keys []}}}
    MaskSpawnGroup=false
    Module.Enable Lathe Profile=false
    Module.Override Randomness=false
    Module.Surface Only=false
   )
   &AddVelocityFromPoint(
    Module.Use Falloff Distance=false
    Origin Offset=<0 0 -10>
    Velocity Strength=&UniformRangedFloat(Minimum=80, Maximum=200)
   )
   &InitialMeshOrientation(
    Rotation=&UniformRangedVector(Minimum=<0 0 0>, Maximum=<0 0 1>)
    Use Orientation Vector=true
    Use Rotation Vector=true
   )
  ]
  particleUpdateScript [
   &ParticleState()
   &GravityForce(Gravity=&FN_PickVectorFromSpawnGroupCutoff(
     Under Cutoff Vector:Vector3f=<1 1 -500>
     Over Cutoff Vector:Vector3f=<0 0 0>
     SpawnGroup:NiagaraInt32=$Emitter.SpawnGroup
     Cutoff:NiagaraInt32=1
    ))
   &ScaleColor(
    Scale Alpha=&FloatFromCurve(FloatCurve=DataInterface() {propertyValues {Curve {keys [<at 0.4124 1.0113>,<at 1.0011 0.002>]}}})
    ScaleA=true
    ScaleRGB=true
   )
   /SetVariables(Particles.MeshColor:LinearColor=&ScaleLinearColorByCurve(
     InitialValue=<rgba 1 1 1 1>
     ScaleCurve=DataInterface() {propertyValues {
       AlphaCurve {keys [<at 0.2012 1>,<at 0.5654 0>]}
       BlueCurve {keys [<at 0.3226 1>,<at 0.9987 0>]}
       GreenCurve {keys [<at 0.3226 1>,<at 0.9987 0>]}
       RedCurve {keys [<at 0.3226 1>,<at 0.9987 0>]}
      }}
     CurveIndex=$Particles.NormalizedAge
    )) {isSetParameters=true, module=SetVariables_#3}
   &Drag(Use Linear Drag=true, Use Rotational Drag=false)
   &SolveForcesAndVelocity(Clamp Velocity=false, Limit Acceleration=false)
   &ScaleSpriteSize(Scale Factor=&Vector2DFromFloat(Value=&FN_PickFloatFromSpawnGroupCutoff(
      Over Cutoff Float:NiagaraFloat=0
      Under Cutoff Float:NiagaraFloat=&FloatFromCurve(FloatCurve=DataInterface() {propertyValues {Curve {keys [
           {interpMode=RCIM_Cubic, tangentMode=RCTM_Break, time=0.5, value=0.9489}
           {arriveTangent=-3.7049, interpMode=RCIM_Cubic, tangentMode=RCTM_Break, time=1.0011, value=0.002}
          ]}}})
      SpawnGroup:NiagaraInt32=$Emitter.SpawnGroup
      Cutoff:NiagaraInt32=2
     )))
   &ScaleMeshSize(Scale Factor=&VectorFromCurve(VectorCurve=DataInterface() {propertyValues {
       XCurve {keys [{interpMode=RCIM_Cubic, leaveTangent=4.8413, tangentMode=RCTM_Break, time=0, value=0}, <at 0.7127 0.5793>]}
       YCurve {keys [{interpMode=RCIM_Cubic, leaveTangent=4.8413, tangentMode=RCTM_Break, time=0, value=0}, <at 0.7127 0.5793>]}
       ZCurve {keys [
         {interpMode=RCIM_Cubic, leaveTangent=7.2542, tangentMode=RCTM_Break, time=0, value=0}
         {arriveTangent=-1.166, interpMode=RCIM_Cubic, leaveTangent=-1.166, tangentMode=RCTM_User, time=0.0765, value=1.1529}
         <at 0.9989 0>
        ]}
      }}))
   &UpdateMeshOrientation(Rotation Rate=0.2, Rotation Vector:Vector3f=<0 0 1>)
  ]
  rendererClasses [&NiagaraSpriteRendererProperties]
  renderers [{rendererClass=&NiagaraSpriteRendererProperties}]
 }
 ImpactDropletsSmall {
  emitterUpdateScript [&EmitterState()]
  particleSpawnScript [
   &InitializeParticle(
    Color=<rgba 1 1 1 2.5>
    Lifetime=&UniformRangedFloat(Minimum=0.2, Maximum=0.6)
    Module.Write Color=true
    Module.Write Lifetime=true
    Module.Write Mass=true
    Module.Write Material Random=true
    Module.Write Position=false
    Module.Write Scale=true
    Module.Write SpriteRotation=false
    Module.Write SpriteSize=true
    Position=&SimulationPosition()
    Sprite Size=&Multiply_Vector2(
     A=<3 5.5>
     B=&Vector2DFromFloat(Value=&UniformRangedFloat(Minimum=0.05, Maximum=1.1))
    )
   )
   &AddVelocityInCone(
    Cone Angle=80
    Cone Axis=<0 0 1>
    Module.Use Velocity Falloff On Cone Axis=false
    Velocity Strength=&FN_PickFloatFromSpawnGroupCutoff(
     Over Cutoff Float:NiagaraFloat=0
     Under Cutoff Float:NiagaraFloat=&UniformRangedFloat(Minimum=50, Maximum=250)
     SpawnGroup:NiagaraInt32=$Emitter.SpawnGroup
     Cutoff:NiagaraInt32=2
    )
   ) {enabled=false}
   /SetVariables(Particles.PositionSplash:NiagaraPosition=$Engine.Owner.Position) {isSetParameters=true, module=SetVariables_#4}
   /SetVariables(Particles.Scale=&FN_PickVectorFromSpawnGroupCutoff(
     Under Cutoff Vector:Vector3f=<0 0 0>
     Over Cutoff Vector:Vector3f=<1 1 1>
     SpawnGroup:NiagaraInt32=$Emitter.SpawnGroup
     Cutoff:NiagaraInt32=2
    )) {isSetParameters=true, module=SetVariables_#1}
   /SetVariables() {isSetParameters=true, module=SetVariables_#2}
   &CylinderLocation(
    Cylinder Height=0
    Cylinder Radius=8
    Lathe Profile=DataInterface() {propertyValues {Curve {keys []}}}
    MaskSpawnGroup=false
    Module.Enable Lathe Profile=false
    Module.Override Randomness=false
    Module.Surface Only=false
   )
   &AddVelocityFromPoint(
    Module.Use Falloff Distance=false
    Origin Offset=<0 0 -15>
    Velocity Strength=&UniformRangedFloat(Minimum=80, Maximum=200)
   )
   &InitialMeshOrientation(
    Rotation=&UniformRangedVector(Minimum=<0 0 0>, Maximum=<0 0 1>)
    Use Orientation Vector=true
    Use Rotation Vector=true
   )
  ]
  particleUpdateScript [
   &ParticleState()
   &GravityForce(Gravity=&FN_PickVectorFromSpawnGroupCutoff(
     Under Cutoff Vector:Vector3f=<1 1 -500>
     Over Cutoff Vector:Vector3f=<0 0 0>
     SpawnGroup:NiagaraInt32=$Emitter.SpawnGroup
     Cutoff:NiagaraInt32=1
    ))
   &ScaleColor(
    Scale Alpha=&FloatFromCurve(FloatCurve=DataInterface() {propertyValues {Curve {keys [<at 0.4124 1.0113>,<at 1.0011 0.002>]}}})
    ScaleA=true
    ScaleRGB=true
   )
   /SetVariables(Particles.MeshColor:LinearColor=&ScaleLinearColorByCurve(
     InitialValue=<rgba 1 1 1 1>
     ScaleCurve=DataInterface() {propertyValues {
       AlphaCurve {keys [<at 0.2012 1>,<at 0.5654 0>]}
       BlueCurve {keys [<at 0.3226 1>,<at 0.9987 0>]}
       GreenCurve {keys [<at 0.3226 1>,<at 0.9987 0>]}
       RedCurve {keys [<at 0.3226 1>,<at 0.9987 0>]}
      }}
     CurveIndex=$Particles.NormalizedAge
    )) {isSetParameters=true, module=SetVariables_#3}
   &Drag(Use Linear Drag=true, Use Rotational Drag=false)
   &SolveForcesAndVelocity(Clamp Velocity=false, Limit Acceleration=false)
   &ScaleSpriteSize(Scale Factor=&Vector2DFromFloat(Value=&FN_PickFloatFromSpawnGroupCutoff(
      Over Cutoff Float:NiagaraFloat=0
      Under Cutoff Float:NiagaraFloat=&FloatFromCurve(FloatCurve=DataInterface() {propertyValues {Curve {keys [
           {interpMode=RCIM_Cubic, tangentMode=RCTM_Break, time=0.5, value=0.9489}
           {arriveTangent=-3.7049, interpMode=RCIM_Cubic, tangentMode=RCTM_Break, time=1.0011, value=0.002}
          ]}}})
      SpawnGroup:NiagaraInt32=$Emitter.SpawnGroup
      Cutoff:NiagaraInt32=2
     )))
   &ScaleMeshSize(Scale Factor=&VectorFromCurve(VectorCurve=DataInterface() {propertyValues {
       XCurve {keys [{interpMode=RCIM_Cubic, leaveTangent=4.8413, tangentMode=RCTM_Break, time=0, value=0}, <at 0.7127 0.5793>]}
       YCurve {keys [{interpMode=RCIM_Cubic, leaveTangent=4.8413, tangentMode=RCTM_Break, time=0, value=0}, <at 0.7127 0.5793>]}
       ZCurve {keys [
         {interpMode=RCIM_Cubic, leaveTangent=7.2542, tangentMode=RCTM_Break, time=0, value=0}
         {arriveTangent=-1.166, interpMode=RCIM_Cubic, leaveTangent=-1.166, tangentMode=RCTM_User, time=0.0765, value=1.1529}
         <at 0.9989 0>
        ]}
      }}))
   &UpdateMeshOrientation(Rotation Rate=0.2, Rotation Vector:Vector3f=<0 0 1>)
  ]
  rendererClasses [&NiagaraSpriteRendererProperties]
  renderers [{rendererClass=&NiagaraSpriteRendererProperties}]
 }
 Ripple {
  bEnabled=false
  emitterUpdateScript [&EmitterState(Life Cycle Mode=::NewEnumerator1, Loop Duration=1, UseLoopDelay=false)]
  particleSpawnScript [
   &InitializeParticle(
    Color=<rgba 1 1 1 0.25>
    Lifetime=1.25
    Module.Write Color=true
    Module.Write Lifetime=true
    Module.Write Mass=true
    Module.Write Material Random=true
    Module.Write Position=true
    Module.Write Scale=false
    Module.Write SpriteRotation=false
    Module.Write SpriteSize=true
    Position=&Add_Vector(A=Linked() {linkedVariable {name=Engine.Owner.Position, type=NiagaraPosition}}, B=<0 0 -5>)
    Sprite Size=&Vector2DFromFloat(Value=35)
   )
   /SetVariables(Particles.SpriteFacing=<0 0 1>) {isSetParameters=true}
   &DynamicMaterialParameters(
    Index 1 Param 3=200
    Module.Param0WriteEnabled1=false
    Module.Param1WriteEnabled1=false
    Module.Param2WriteEnabled1=true
    Module.Param3WriteEnabled1=false
    Write Parameter Index 0=false
    Write Parameter Index 1=true
   )
  ]
  particleUpdateScript [
   &ParticleState()
   &DynamicMaterialParameters(
    Index 0 Param 1=&FloatFromCurve(FloatCurve=DataInterface() {propertyValues {Curve {keys [{arriveTangent=6.0763, interpMode=RCIM_Cubic, leaveTangent=6.0763, tangentMode=RCTM_User, time=-0.0046, value=1}, <at 1.0011 3>]}}})
    Index 0 Param 2=&FloatFromCurve(FloatCurve=DataInterface() {propertyValues {Curve {keys [
         {arriveTangent=1.2087, interpMode=RCIM_Cubic, leaveTangent=1.2087, tangentMode=RCTM_User, time=-0.002, value=0.3756}
         <at 1.0011 1.4657 RCIM_Cubic>
        ]}}})
    Index 0 Param 3=&FloatFromCurve(FloatCurve=DataInterface() {propertyValues {Curve {keys [<at -0.0046 1>,<at 1.0011 0.5>]}}})
    Index 2 Param 1=&FloatFromCurve(FloatCurve=DataInterface() {propertyValues {Curve {keys [<at -0.0046 1.4697>,<at 0.369 1.481>,<at 1.0011 2>]}}})
    Index 2 Param 2=&FloatFromCurve(FloatCurve=DataInterface() {propertyValues {Curve {keys [<at -0.0046 0.1>,<at 0.1026 4.2375>,<at 1.0011 0.3>]}}})
    Index 2 Param 3=&FloatFromCurve(FloatCurve=DataInterface() {propertyValues {Curve {keys [
         {arriveTangent=-2.2186, interpMode=RCIM_Cubic, leaveTangent=-2.2186, tangentMode=RCTM_User, time=-0.0046, value=1.0113}
         <at 1.0011 0.002 RCIM_Cubic>
        ]}}})
    Index 2 Param 4=&FloatFromCurve(FloatCurve=DataInterface() {propertyValues {Curve {keys [
         {arriveTangent=-1.9991, interpMode=RCIM_Cubic, time=0, value=1}
         {arriveTangent=-1.3497, interpMode=RCIM_Cubic, leaveTangent=-1.3497, tangentMode=RCTM_User, time=0.1761, value=0.423}
         <at 1 0 RCIM_Cubic>
        ]}}})
    Module.Param0WriteEnabled=true
    Module.Param0WriteEnabled2=true
    Module.Param1WriteEnabled=true
    Module.Param1WriteEnabled2=true
    Module.Param2WriteEnabled=true
    Module.Param2WriteEnabled2=true
    Module.Param3WriteEnabled=true
    Module.Param3WriteEnabled2=true
    Write Parameter Index 2=true
   ) {module=DynamicMaterialParameters001}
   &SolveForcesAndVelocity(Clamp Velocity=false, Limit Acceleration=false)
   &ScaleColor(
    Scale Alpha=&FloatFromCurve(FloatCurve=DataInterface() {propertyValues {Curve {keys [{interpMode=RCIM_Cubic, leaveTangent=-2.8107, tangentMode=RCTM_Break, time=0.1163, value=1.0113}, <at 1.0011 0.002>]}}})
    ScaleA=true
    ScaleRGB=true
   )
   &ScaleSpriteSize(Scale Factor=&Vector2DFromFloat(Value=&FloatFromCurve(FloatCurve=DataInterface() {propertyValues {Curve {keys [<at -0.0046 1.0113>,<at 0.0708 1.5192>,<at 1 1>]}}})))
  ]
  rendererClasses [&NiagaraSpriteRendererProperties]
  renderers [{rendererClass=&NiagaraSpriteRendererProperties}]
 }
 UnderwaterBubbles {
  emitterUpdateScript [&EmitterState(Life Cycle Mode=::NewEnumerator1, Loop Duration=1, UseLoopDelay=false)]
  particleSpawnScript [
   &InitializeParticle(
    Color=<rgba 1 1 1 2>
    Lifetime=&UniformRangedFloat(Minimum=0.5, Maximum=2)
    Module.Write Color=true
    Module.Write Lifetime=true
    Module.Write Mass=true
    Module.Write Material Random=true
    Module.Write Position=false
    Module.Write Scale=false
    Module.Write SpriteRotation=false
    Module.Write SpriteSize=true
    Position=&SimulationPosition()
    Sprite Size=&Vector2DFromFloat(Value=&UniformRangedFloat(Minimum=0.5, Maximum=3.5))
   )
   /SetVariables(Particles.RandomPosNeg:NiagaraFloat=&UniformRangedFloat(Minimum=-1, Maximum=1)) {isSetParameters=true}
   &AddVelocity(Velocity=&UniformRangedVector(Minimum=<-5 -5 -25>, Maximum=<5 5 -12>))
  ]
  particleUpdateScript [
   &ParticleState()
   /Niagara/Modules/Update/Forces/CurlNoiseForce(Noise Frequency=0.5, Noise Quality / Cost=::NewEnumerator0, Noise Strength=50, Pan Noise=false)
   &GravityForce(Gravity=<0 0 32>)
   &SolveForcesAndVelocity(Clamp Velocity=false, Limit Acceleration=false)
   &UpdateMeshOrientation(
    Rotation Rate:NiagaraFloat=&Multiply_Float(A=0.035, B=$Particles.RandomPosNeg)
    Rotation Vector:Vector3f=&VectorFromFloat(Value=1)
   )
   &ScaleSpriteSize(Scale Factor=&Vector2DFromFloat(Value=&FloatFromCurve(FloatCurve=DataInterface() {propertyValues {Curve {keys [
          <at 0 0 RCIM_Cubic>
          {arriveTangent=0.002, interpMode=RCIM_Cubic, leaveTangent=-0.0489, tangentMode=RCTM_Break, time=0.1999, value=0.8838}
          {arriveTangent=-0.077, interpMode=RCIM_Cubic, leaveTangent=-0.077, tangentMode=RCTM_User, time=0.9997, value=0.0368}
         ]}}})))
   &DynamicMaterialParameters(
    Index 0 Param 1=1
    Index 0 Param 2=1
    Index 0 Param 3=1
    Index 0 Param 4=&FloatFromCurve(FloatCurve=DataInterface() {propertyValues {Curve {keys [<at 0 0>,<at 1 1>]}}})
    Module.Param0WriteEnabled=true
    Module.Param1WriteEnabled=true
    Module.Param2WriteEnabled=true
    Module.Param3WriteEnabled=true
   )
  ]
  rendererClasses [&NiagaraSpriteRendererProperties]
  renderers [{rendererClass=&NiagaraSpriteRendererProperties}]
 }
}
summary {
 systemName=NS_Donut_Water_Drips_And_Splash
 emitters [
  {emitterName=DrippingWater, rendererClasses [&NiagaraSpriteRendererProperties]}
  {emitterName=CrownSplashMesh, rendererClasses [&NiagaraMeshRendererProperties]}
  {emitterName=UnderwaterBubbles, rendererClasses [&NiagaraSpriteRendererProperties]}
  {bEnabled=false, emitterName=Ripple, rendererClasses [&NiagaraSpriteRendererProperties]}
  {emitterName=ImpactDroplets, rendererClasses [&NiagaraSpriteRendererProperties]}
  {emitterName=ImpactDropletsSmall, rendererClasses [&NiagaraSpriteRendererProperties]}
 ]
 userVariables [User.DonutWaterLocation {type=Vector3f, description="", defaultValue {struct=Vector3f, value=<0 0 0>}}]
}
