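# Niagara sprite system for rising bubbles, named as the water mask bubbles of the SpyBase front-end Donut effects.
# One emitter spawns in a flat disc sized by a user parameter, at a random rate whose probability falls over the emitter loop.
# Bubbles drift upward under random acceleration and drag, pop in and swell before death, while a dynamic material parameter ramps over lifetime.
# The height-limit step is disabled, and StopHeight appears unused.
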NiagaraSystem /BRCosmetics/Effects/Fort_Effects/FrontEnd/SpyBase/Particles/Donut/NS_Donut_Water_Mask_Bubbles
dependencies {usedDataInterfaces [&NiagaraDataInterfaceCurve], usedDynamicInputs [], usedModules [&SystemState]}
emitters {Empty002 {
  emitterUpdateScript [
   &EmitterState(Life Cycle Mode=::NewEnumerator1, Loop Behavior=::NewEnumerator1, Loop Duration=2.5, UseLoopDelay=false)
   &SpawnRate(
    Spawn Probability=&FloatFromCurve(
     FloatCurve=DataInterface() {propertyValues {Curve {keys [{arriveTangent=-2.786, interpMode=RCIM_Cubic, leaveTangent=-2.786, tangentMode=RCTM_User, time=-0.0046, value=1.0113}, <at 1.0011 0.002>]}}}
     CurveIndex=$Emitter.NormalizedLoopAge
    )
    SpawnRate=&UniformRangedFloat(Minimum=2, Maximum=160)
    Use Spawn Probability=true
   )
  ]
  particleSpawnScript [
   /Niagara/Modules/Spawn/Initialization/InitializeParticle(
    Color=<rgba 1 1 1 15>
    Lifetime=$User.Lifetime
    Mesh Scale=<1.4 1.4 1.4>
    Module.Write Color=true
    Module.Write Lifetime=true
    Module.Write Mass=true
    Module.Write Material Random=true
    Module.Write Position=false
    Module.Write Scale=false
    Module.Write SpriteRotation=false
    Module.Write SpriteSize=true
    Position=&SimulationPosition()
    Sprite Size=&Vector2DFromFloat(Value=&UniformRangedFloat(Minimum=0.125, Maximum=3.5))
   )
   /SetVariables(Particles.RandomPosNeg:NiagaraFloat=&UniformRangedFloat(Minimum=-1, Maximum=1)) {isSetParameters=true}
   &AddVelocity(Velocity=&UniformRangedVector(Minimum=<-3 -3 25>, Maximum=<3 3 50>))
   &CylinderLocation(
    Cylinder Height=0
    Cylinder Origin:NiagaraPosition=$Engine.Owner.Position
    Cylinder Radius=$User.CylinderLocationWidth
    Lathe Profile=DataInterface() {propertyValues {Curve {keys []}}}
    MaskSpawnGroup=false
    Module.Enable Lathe Profile=false
    Module.Override Randomness=false
    Module.Surface Only=false
   )
  ]
  particleUpdateScript [
   &ParticleState()
   &AccelerationForce(Acceleration=&UniformRangedVector(Minimum=<-30 -30 50>, Maximum=<30 30 1>))
   &Drag(Use Linear Drag=true, Use Rotational Drag=false)
   &SolveForcesAndVelocity(Clamp Velocity=false, Limit Acceleration=false)
   /Niagara/Modules/Update/Orientation/V2/UpdateMeshOrientation(
    Rotation Rate:NiagaraFloat=&Multiply_Float(A=0.035, B=$Particles.RandomPosNeg)
    Rotation Vector:Vector3f=&VectorFromFloat(Value=1)
   )
   /SetVariables(Particles.Position=&self:LimitPositionAboveZero*()) {enabled=false, isSetParameters=true}
   &ScaleSpriteSize(Scale Factor=&Vector2DFromFloat(Value=&FloatFromCurve(FloatCurve=DataInterface() {propertyValues {Curve {keys [
          <at 0 0 RCIM_Cubic>
          {arriveTangent=0.002, interpMode=RCIM_Cubic, leaveTangent=-0.0489, tangentMode=RCTM_Break, time=0.0313, value=0.9069}
          {arriveTangent=-0.077, interpMode=RCIM_Cubic, leaveTangent=-0.077, tangentMode=RCTM_User, time=0.9114, value=0.8888}
          {arriveTangent=4.9648, interpMode=RCIM_Cubic, leaveTangent=4.9648, tangentMode=RCTM_User, time=1, value=1.2}
         ]}}})))
   &DynamicMaterialParameters(
    Index 0 Param 1=1
    Index 0 Param 2=1
    Index 0 Param 3=1
    Index 0 Param 4=&FloatFromCurve(FloatCurve=DataInterface() {propertyValues {Curve {keys [<at 0 0>,<at 1 1>]}}})
    Module.Param0WriteEnabled=true
    Module.Param1WriteEnabled=true
    Module.Param2WriteEnabled=true
    Module.Param3WriteEnabled=true
   )
  ]
  rendererClasses [&NiagaraSpriteRendererProperties]
  renderers [{rendererClass=&NiagaraSpriteRendererProperties}]
 }}
summary {
 systemName=NS_Donut_Water_Mask_Bubbles
 emitters [{emitterName=Empty002, rendererClasses [&NiagaraSpriteRendererProperties]}]
 userVariables [
  User.CylinderLocationWidth {type=NiagaraFloat, defaultValue=2, description=""}
  User.Lifetime {type=NiagaraFloat, defaultValue=5, description=""}
  User.StopHeight {type=NiagaraFloat, defaultValue=25, description=""}
 ]
}
